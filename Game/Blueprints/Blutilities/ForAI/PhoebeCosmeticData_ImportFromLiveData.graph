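# Editor utility that imports live cosmetic usage data from a chosen CSV into selected AI bot cosmetic library DataTables.
# It optionally backs up the "New" library into "Previous" first.
# CSV loading filters by usage, name, prefix, IP owner, tags and over-budget tech anim cosmetics.
# Per asset type, it scales counts to clamped 0-100 weights against a reference entry, then refills, saves and re-exports each table.
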EditorUtilityBlueprint /Game/Blueprints/Blutilities/ForAI/PhoebeCosmeticData_ImportFromLiveData
graphs [
 BuildOverBudgetCosmeticList {nodes [
   BuildOverBudgetCosmeticList() {is=title K2Node_FunctionEntry}
   > Clear(TargetArray=Get FinalList() -> (FinalList:Array of Strings))
   > n1 = Branch(Condition="float > float"(A=Get CosmeticMaxBudgetMS() -> (CosmeticMaxBudgetMS:double)))
   n2 = Get CosmeticBudgetDataTable() -> (CosmeticBudgetDataTable:Data Table Object Reference)
   > n3 = GetDataTableRowNames(Table=@n2)
   > n4 = For Each Loop(Array:Array of Names=@n3) -> (Array Element:Name)
   Return Node!(execute=[@n4.Completed, @n1.else], OverBudgetList:Array of Strings=Get FinalList())
   > n5 = Get Data Table Row!(execute:Exec, DataTable:Data Table Object Reference=@n2, RowName:Name=@n4) -> (then:Exec, RowNotFound:Exec, ReturnValue:Cosmetic Tech Anim Budget Table Row Structure)
   > Branch(Condition="float >= float"(
     A=Break Cosmetic Tech Anim Budget Table Row!(CosmeticTechAnimBudget_TableRow:"Cosmetic Tech Anim Budget Table Row Structure (by ref)"=@n5) -> ("TOTAL-->_13_#1":double=~) {is=title K2Node_BreakStruct}
     B=Get CosmeticMaxBudgetMS()
    ))
   > Add(TargetArray=Get FinalList(), NewItem="To String (Name)"(InName=@n4)) {is=title K2Node_CallArrayFunction}
  ]}
 ComputeAssetWeightList {nodes [
   n1 = ComputeAssetWeightList!() -> (then, Usage Counts:Array of Integers) {is=title K2Node_FunctionEntry}
   n2 = Length(TargetArray:Array of Integers=@n1)
   > Resize!(execute, TargetArray:Array of Integers=Get LocalWeightList() -> (LocalWeightList:Array of Integers), Size:int=@n2) -> (then) {is=title K2Node_CallArrayFunction}
   n3 = int / int(A=@n2, B=2)
   > n4 = Set HighUsageCount(HighUsageCount:int="Get (a copy)"(Array:Array of Integers=@n1, Dimension 1="Min (Integer)"(A=@n3, B=100) -> (ReturnValue=0)) -> (Output:int)) -> (Output_Get:int=0)
   > Compare Int!(Exec:Exec, Input:int=@n4, CompareWith:int) -> (">":Exec, "==":Exec, "<":Exec) {is=title K2Node_MacroInstance}
   > n5 = For Each Loop(Array:Array of Integers=@n1) -> (Array Element:int)
   Return Node!(execute=@n5.Completed, Weight List:Array of Integers=Get LocalWeightList())
   n6 = float / float(
    A=float * float(A="To Float (Integer)"(InInt=@n5), B=100) {is=title K2Node_PromotableOperator}
    B:int=Get HighUsageCount() -> (HighUsageCount:int=0)
   )
   > Set Array Elem!(
    execute
    TargetArray:Array of Integers=Get LocalWeightList()
    Index:int=@n5.Array Index
    Item:"Integer (by ref)"="Clamp (Integer)"(Value=Ceil(A=@n6) {is=title K2Node_CallFunction}, Min=0, Max=100) -> (ReturnValue=0)
    bSizeToFit:bool=false
   ) -> (then)
  ]}
 ComputeForbiddenAssetStrings {nodes [
   ComputeForbiddenAssetStrings() {is=title K2Node_FunctionEntry}
   > Clear(TargetArray=Get FinalList())
   > Append Array!(
    execute
    TargetArray=Get FinalList()
    SourceArray:Array of Strings=Get Forbidden Asset Strings() -> (Forbidden Asset Strings:Array of Strings)
   ) -> (then)
   > n1 = BuildOverBudgetCosmeticList!(execute, self) -> (then, OverBudgetList:Array of Strings) {is=title K2Node_CallFunction}
   > Append Array(TargetArray=Get FinalList(), SourceArray=@n1)
   > Return Node!(execute, ForbiddenAssetStrings:Array of Strings=Get FinalList())
  ]}
 EventGraph {links [], nodes []}
 ImportFromLiveData {nodes [
   ImportFromLiveData() {is=title K2Node_FunctionEntry}
   > n1 = GetSelectedAssets()
   > n2 = For Each Loop(Array:Array of Object References=@n1) -> (Array Element:Object Reference)
   > n3 = Cast To FortAthenaAIBotCosmeticData!(execute, Object=@n2) -> (then, CastFailed, AsFort Athena AIBot Cosmetic Data:Fort Athena AIBot Cosmetic Data Object Reference) {is=title K2Node_DynamicCast}
   > n4 = RetrieveCosmeticLibraryDataToFIll!(execute, self, CosmeticData:Fort Athena AIBot Cosmetic Data Object Reference=@n3) -> (then, Data:Fort Athena AIBot Cosmetic Library Data Object Reference, DataTransfered:bool)
   > n5 = Is Valid(InputObject=@n4)
   > n6 = RetrieveCSVFile!(execute, self, Cosmetic Library Data:Fort Athena AIBot Cosmetic Library Data Object Reference=@n4) -> (then, FilePath:String)
   > n7 = ComputeForbiddenAssetStrings!(execute, self) -> (then, ForbiddenAssetStrings:Array of Strings) {is=title K2Node_CallFunction}
   > n8 = LoadDataFromCSV!(
    execute
    FilePath:String=@n6
    AssetNameColumnTitle:String=Get Asset Name Column Title() -> (Asset Name Column Title:String)
    AssetTypeColumnTitle:String=Get Asset Type Column Title() -> (Asset Type Column Title:String)
    UsageCountColumnTitle:String=Get Usage Count Column Title() -> (Usage Count Column Title:String)
    MinUsageCount:int=Get Minimum Usage Count() -> (Minimum Usage Count:int=0)
    ForbiddenAssetStrings:Array of Strings=@n7
    ForbiddenAssetPrefix:Array of Strings=Get Forbidden Asset Prefix() -> (Forbidden Asset Prefix:Array of Strings)
    ForbiddenIPOwner:Array of Names=Get Forbidden IP Owners() -> (Forbidden IP Owners:Array of Names)
    ForbiddenTags:"Gameplay Tag Container Structure (by ref)"=ComputeExcludedCosmeticTags!(self:Fort Athena AIBot Cosmetic Library Data Object Reference=@n4) -> (ReturnValue:Gameplay Tag Container Structure)
   ) -> (
    then
    AssetNames:Array of Strings
    AssetTypes:Array of Strings
    AssetUsageCounts:Array of Integers
    AssetTypeSplitIndices:Array of Integers
    ReturnValue:bool
   )
   > n9 = Branch(Condition=@n8.ReturnValue)
   > n10 = For Each Loop(Array:Array of Integers=@n8.AssetTypeSplitIndices) -> (Array Element:int)
   n11 = GetDisplayName(Object=@n4)
   n12 = SelectString(A=" and a backup was saved", bPickA=@n4.DataTransfered)
   Show Message Dialog(
    execute=@n10.Completed
    Title=Get End Dialog Title() -> (End Dialog Title:Text)
    Message="To Text (String)"(InString=Append!(A="Data has been loaded from ", B=@n6, C:String=" into ", D:String=@n11, E:String=@n12) -> (ReturnValue))
    MessageType=Ok
    DefaultValue=Ok
    MessageCategory=Warning
   )
   ProcessAssetType!(
    execute=@n10.LoopBody
    self
    InAssetNames:Array of Strings=@n8
    InAssetTypes:Array of Strings=@n8.AssetTypes
    InAssetUsageCounts:Array of Integers=@n8.AssetUsageCounts
    InAssetTypeSplitIndices:Array of Integers=@n8.AssetTypeSplitIndices
    CurrentSplitIndex:int=@n10.Array Index
    Data:Fort Athena AIBot Cosmetic Library Data Object Reference=@n4
   ) -> (then)
   Show Message Dialog(
    execute=[@n9.else, @n5.Is Not Valid]
    Title=Get FailDialogTitle() -> (FailDialogTitle:Text)
    Message=Get FailDialogMsg() -> (FailDialogMsg:Text)
    MessageType=Ok
    DefaultValue=Ok
    MessageCategory=Error
   )
  ]}
 ProcessAssetType {nodes [
   n1 = ProcessAssetType!() -> (
    then
    InAssetNames:Array of Strings
    InAssetTypes:Array of Strings
    InAssetUsageCounts:Array of Integers
    InAssetTypeSplitIndices:Array of Integers
    CurrentSplitIndex:int
    Data:Fort Athena AIBot Cosmetic Library Data Object Reference
   ) {is=title K2Node_FunctionEntry}
   > n2 = SplitDataArrayByType!(
    execute
    InAssetNames:Array of Strings=@n1
    InAssetTypes:Array of Strings=@n1.InAssetTypes
    InAssetUsageCounts:Array of Integers=@n1.InAssetUsageCounts
    InAssetTypeSplitIndices:Array of Integers=@n1.InAssetTypeSplitIndices
    CurrentSplitIndex:int=@n1.CurrentSplitIndex
   ) -> (then, OutAssetType:String, OutAssetNames:Array of Strings, OutAssetUsageCounts:Array of Integers, ReturnValue:bool)
   > Branch(Condition=@n2.ReturnValue)
   > n3 = FindDataTableFromAssetType!(execute, self:Fort Athena AIBot Cosmetic Library Data Object Reference=@n1.Data, AssetType:String=@n2) -> (then, ReturnValue:Data Table Object Reference)
   > Branch(Condition=IsValid(Object=@n3))
   > Set Asset Data Table(Asset Data Table:Data Table Object Reference=@n3) -> (Output_Get:Data Table Object Reference)
   > n4 = ComputeAssetWeightList!(execute, self, Usage Counts:Array of Integers=@n2.OutAssetUsageCounts) -> (then, Weight List:Array of Integers) {is=title K2Node_CallFunction}
   > n5 = GenerateWeightedCSV!(execute, AssetTypeName:String=@n2, AssetNames:Array of Strings=@n2.OutAssetNames, AssetWeight:Array of Integers=@n4) -> (then, CSV:String, ReturnValue:bool)
   > Branch(Condition=@n5.ReturnValue)
   > n6 = Fill Data Table from CSV String!(
    execute
    DataTable:Data Table Object Reference=Get Asset Data Table() -> (Asset Data Table:Data Table Object Reference)
    CSVString:String=@n5
    ImportRowStruct:Script Struct Object Reference
   ) -> (then, ReturnValue:bool)
   > Branch(Condition=@n6)
   > n7 = SaveLoadedAsset(AssetToSave=Get Asset Data Table(), bOnlyIfIsDirty=true)
   > Branch(Condition=@n7)
   > ExportDataTableToSourceCSV!(execute, DataTable:Data Table Object Reference=Get Asset Data Table(), bCanCheckOutFile:bool=true) -> (then, ReturnValue:bool)
  ]}
 RetrieveCSVFile {nodes [
   n1 = RetrieveCSVFile!() -> (then, Cosmetic Library Data:Fort Athena AIBot Cosmetic Library Data Object Reference) {is=title K2Node_FunctionEntry}
   > n2 = OpenCSVFileDialog!(execute, TitleDetails:String=GetDisplayName(Object=@n1), bAllowMultipleFiles:bool=false) -> (then, ReturnValue:Array of Strings)
   > n3 = Branch(Condition=Is Empty!(TargetArray=@n2) -> (ReturnValue:bool))
   > Return Node!(execute, FilePath:String)
   Return Node(execute=@n3.else, FilePath="Get (a ref)"(Array:Array of Strings=@n2, Dimension 1=0) -> (Output:"String (by ref)"))
  ]}
 RetrieveCosmeticLibraryDataToFIll {nodes [
   n1 = RetrieveCosmeticLibraryDataToFIll!() -> (then, CosmeticData:Fort Athena AIBot Cosmetic Data Object Reference) {is=title K2Node_FunctionEntry}
   > n2 = FindLibraryDataFromName!(execute, self:Fort Athena AIBot Cosmetic Data Object Reference=@n1, PartialName:String=New) -> (then, ReturnValue:Fort Athena AIBot Cosmetic Library Data Soft Object Reference)
   > n3 = LoadAsset_Blocking(Asset=@n2)
   > n4 = Cast To FortAthenaAIBotCosmeticLibraryData!(execute, Object=@n3) -> (then, CastFailed, AsFort Athena AIBot Cosmetic Library Data:Fort Athena AIBot Cosmetic Library Data Object Reference) {is=title K2Node_DynamicCast}
   > Set New Library(New Library:Fort Athena AIBot Cosmetic Library Data Object Reference=@n4) -> (Output_Get:Fort Athena AIBot Cosmetic Library Data Object Reference)
   > n5 = Show Message Dialog(
    Title=Get Dialog Title() -> (Dialog Title:Text)
    Message=Get Dialog Message() -> (Dialog Message:Text)
    MessageType=YesNoCancel
    DefaultValue=Yes
    MessageCategory=Warning
   )
   > n6 = Switch on EAppReturnType(Selection=@n5)
   Return Node!(
    execute=[@n4.CastFailed, @n9.CastFailed, @n11.else, @n6.Cancel]
    Data:Fort Athena AIBot Cosmetic Library Data Object Reference
    DataTransfered:bool=false
   )
   > Return Node(Data=Get New Library() -> (New Library:Fort Athena AIBot Cosmetic Library Data Object Reference), DataTransfered=false)
   n7 = FindLibraryDataFromName(execute=@n6.Yes, self=@n1, PartialName=Previous)
   > n8 = LoadAsset_Blocking(Asset=@n7)
   > n9 = Cast To FortAthenaAIBotCosmeticLibraryData(Object=@n8)
   > n10 = FillDataTableValuesFromSourceLibrary!(
    execute
    self:Fort Athena AIBot Cosmetic Library Data Object Reference=@n9
    SourceLibrary:Fort Athena AIBot Cosmetic Library Data Object Reference=Get New Library()
   ) -> (then, ReturnValue:bool)
   > n11 = Branch(Condition=@n10)
   n12 = RetrieveDataTables!(self:Fort Athena AIBot Cosmetic Library Data Object Reference=@n9) -> (ReturnValue:Array of Data Table Object References)
   > SaveLoadedAssets(AssetsToSave=@n12, bOnlyIfIsDirty=true)
   > n13 = For Each Loop(Array:Array of Data Table Object References=@n12) -> (Array Element:Data Table Object Reference)
   Return Node(execute=@n13.Completed, Data=Get New Library(), DataTransfered=true)
   > ExportDataTableToSourceCSV(DataTable=@n13, bCanCheckOutFile=true)
  ]}
]
variables [
 Asset Name Column Title {type=string, default=Item}
 Asset Type Column Title {type=string, default=Item Type}
 CosmeticBudgetDataTable {type=DataTable, default=&/DT_CosmeticTechAnimBudget}
 CosmeticMaxBudgetMS {type=number, default=1.75}
 Forbidden Asset Prefix {type="array<string>", default [spid_, emoji_, spray_, athenacharacteritemdefinition_]}
 Forbidden Asset Strings {
  type="array<string>"
  default [
   unknown
   null
   defaultcontrail
   defaultpickaxe
   defaultglider
   bid_stwheronodefaultbackpack
   _athena_commando_f_rebirthdefault
   _athena_commando_m_rebirthdefault
   _athena_commando_f_prime
   _athena_commando_m_prime
   Character_PrimeRedux
   pickaxe_lime
   pickaxe_beanhands
   npc
  ]
 }
 Forbidden IP Owners {type="array<string>", default [Lucas,20CS,Marvel,Disney,Fox,Pixar]}
 Minimum Usage Count {type=integer, default=100}
 Usage Count Column Title {type=string, default=Selected Metric}
]
